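; Burnt Tires ceiling corner prop that destroys itself once all of the building actors it is attached to or tracks have died.
; On BeginPlay, the server stores the count of BuildingAttachmentActors and binds to each one's OnDied event.
; Each death increments a counter, and the prop destroys itself when that counter reaches the stored count.
; With an empty list nothing is bound, and the construction script only calls its parent.
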(event EventBeginPlay
  (bind _buildingattachmentactors (Variables|Default|GetBuildingAttachmentActors))
  (bind _returnvalue (Utilities|Array|Length _buildingattachmentactors))
  (Networking|SwitchHasAuthority
    (:Authority
      (if (> _returnvalue 0)
        (Variables|Default|SetAcountCount _returnvalue)
        (for _array_element (Variables|Default|GetBuildingAttachmentActors)
          (EventDispatchers|BindEventtoOnDied _array_element (delegate Custom|Died)))))))

(event Custom|Died
  (params (DamagedActor :type Actor%20Object%20Reference) (Damage :type Float%20%28single-precision%29) (InstigatedBy :type Controller%20Object%20Reference) (DamageCauser :type Actor%20Object%20Reference) (HitLocation :type Vector) (FHitComponent :type Primitive%20Component%20Object%20Reference) (BoneName :type Name) (Momentum :type Vector))
  (Math|Integer|IncrementInt (Variables|Default|GetActorsDiedCount)
    (:%20%20
      (if (>= (Variables|Default|GetActorsDiedCount) (Variables|Default|GetAcountCount))
        (Utilities|IsValid self
          (:Is%20Valid
            (Actor|DestroyActor)))))))

(fn ConstructionScript
  (|CallParentFunction))
