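; Neon sign prop Blueprint for the Helios BurntTires set that boosts how strongly its emissive material feeds into Lumen lighting.
; BoostLumenEmissive gets a dynamic material instance from the static mesh component and sets its LumenEmissiveMultiplier scalar from the LumenBoost variable.
; It runs in the construction script, after the parent's, and again on BeginPlay; the tick and overlap events are empty.
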(event EventBeginPlay
  (CallFunction|BoostLumenEmissive))

(event Collision|EventActorBeginOverlap
  (params (OtherActor :type Actor%20Object%20Reference)))

(event EventTick
  (params (DeltaSeconds :type Float%20%28single-precision%29)))

(fn BoostLumenEmissive
  (Rendering|Material|SetScalarParameterValue (Class|BTechArtGlobalLibrary|GetDynamicMaterialInstance (Variables|Debug|GetStaticMeshComponent)) "LumenEmissiveMultiplier" (Variables|Default|GetLumenBoost) :!class /Script/Engine.MaterialInstanceDynamic))

(fn ConstructionScript
  (|CallParentFunction)
  (CallFunction|BoostLumenEmissive))
