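; BP_VehicleBlocking is a prop Blueprint in the Helios BurntTires set that, by its name, blocks vehicles with a box volume.
; Its construction script sets the extent of its Box component from the BoxBounds variable and updates overlaps.
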(fn ConstructionScript
  (Components|Box|SetBoxExtent (Variables|Default|GetBox) (Variables|Default|GetBoxBounds) true))
